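# Niagara effect for the Phonograph emote that sends music notes, a music-sheet trail and air puffs out of the horn.
# Mesh emitters form pulsing horn air, sprites add lines and wobbling lit notes, and a velocity-inheriting ribbon computes a sine-wave position for the sheet.
# Curves over looped system age swell spawn rates for notes, lines and bursts across roughly ten seconds.
# Platform checks boost alpha on Switch for all but the notes.
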NiagaraSystem /Game/Effects/Fort_Effects/Emotes/Phonograph/NS_Phonograph_Notes
dependencies {
 usedDataInterfaces [&NiagaraDataInterfaceCurve, &NiagaraDataInterfacePlatformSet]
 usedDynamicInputs []
 usedModules [
  &self:SystemSpawnScript.SystemScriptSource.SystemScriptGraph.NiagaraNodeAssignment_1.SetVariables_#1
  &self:SystemSpawnScript.SystemScriptSource.SystemScriptGraph.NiagaraNodeAssignment_0.SetVariables_#2
  &SystemState
  /Niagara/Modules/System/CompleteIfUnused
 ]
}
emitters {
 Horn_Air {
  emitterUpdateScript [
   &EmitterState()
   &SpawnRate(SpawnRate=8, Use Spawn Probability=false)
  ]
  particleSpawnScript [
   /Niagara/Modules/Spawn/Initialization/InitializeParticle(
    Color=<rgba 1 0.8221 0.5 1>
    Lifetime=0.5
    Mesh Scale=&Multiply_Vector(
     A=&Multiply_Vector(
      A=&MakeVector(X=1.5, Y=1.25, Z=&RandomRangeFloat(Minimum=1, Maximum=1.2))
      B=&VectorFromFloat(Value=&RandomRangeFloat(Minimum=0.5, Maximum=0.75))
     )
     B=&VectorFromFloat(Value=1)
    )
    Module.Write Color=true
    Module.Write Lifetime=true
    Module.Write Mass=true
    Module.Write Material Random=true
    Module.Write Position=false
    Module.Write Scale=true
    Module.Write SpriteRotation=false
    Module.Write SpriteSize=false
    Position=<0 0 10>
   )
   /SetVariables(
    Particles.MeshScale_02:Vector3f=<1 1 1>
    Particles.XOffset:NiagaraFloat=&RandomRangeFloat(Minimum=0, Maximum=1)
    Particles.XOffset_02:NiagaraFloat=&RandomRangeFloat(Minimum=0, Maximum=1)
   ) {isSetParameters=true, module=SetVariables_#3}
   &InitialMeshOrientation(
    Orientation Coordinate Space=::Local
    Orientation Vector=<0 0 1>
    Rotation=&RandomRangeVector(Minimum=<0 0 0>, Maximum=<0 0 1>)
    Use Orientation Vector=true
    Use Rotation Vector=true
   )
  ]
  particleUpdateScript [
   &ParticleState()
   &DynamicMaterialParameters(
    Index 0 Param 1=1
    Index 0 Param 2=0.04
    Index 0 Param 3=20
    Index 0 Param 4=1
    Index 1 Param 1=&Add_Float(
     A=$Particles.Initial.XOffset
     B=&FloatFromCurve(FloatCurve=DataInterface() {propertyValues {Curve {keys [<at 0 0>,<at 1 1>]}}}, Scale Curve=2.5)
    )
    Index 1 Param 2=&RandomRangeFloat(Minimum=0, Maximum=1)
    Index 2 Param 1=&Add_Float(
     A=$Particles.Initial.XOffset_02
     B=&FloatFromCurve(FloatCurve=DataInterface() {propertyValues {Curve {keys [<at 0 0>,<at 1 1>]}}}, Scale Curve=2.5)
    )
    Index 2 Param 2=&RandomRangeFloat(Minimum=0, Maximum=1)
    Module.Param0WriteEnabled=true
    Module.Param0WriteEnabled1=true
    Module.Param0WriteEnabled2=true
    Module.Param1WriteEnabled=false
    Module.Param1WriteEnabled1=true
    Module.Param1WriteEnabled2=true
    Module.Param2WriteEnabled=false
    Module.Param2WriteEnabled1=false
    Module.Param2WriteEnabled2=true
    Module.Param3WriteEnabled=true
    Module.Param3WriteEnabled1=false
    Module.Param3WriteEnabled2=true
    Write Parameter Index 1=true
    Write Parameter Index 2=true
   ) {module=DynamicMaterialParameters001}
   &ScaleColor(
    Scale Alpha=&FloatFromCurve(
     FloatCurve=DataInterface() {propertyValues {Curve {keys [
         {arriveTangent=3.2844, interpMode=RCIM_Cubic, leaveTangent=3.2844, tangentMode=RCTM_User, time=0, value=0}
         <at 0.2049 0.8535 RCIM_Cubic>
         {arriveTangent=-3.5444, interpMode=RCIM_Cubic, leaveTangent=-3.5444, tangentMode=RCTM_User, time=1, value=0}
        ]}}}
     Scale Curve=&MakeCustomFloatFromBool(
      Bool=&IsPlatformSetActive(New Platform Set=DataInterface() {propertyValues {Platforms {qualityLevelMask=-32, deviceProfileStates [{profileName=Switch, qualityLevelMask=4294967295, setQualityLevelMask=4294967295}]}}})
      True Float=7
      False Float=1
     )
    )
    ScaleA=true
    ScaleRGB=false
   )
   &UpdateMeshOrientation(
    Rotation Coordinate Space=::NewEnumerator3
    Rotation Rate=&RandomRangeFloat(Minimum=-1, Maximum=1)
    Rotation Vector=<0 0 1>
   )
   &ScaleMeshSize(Scale Factor=&VectorFromFloat(Value=&FloatFromCurve(FloatCurve=DataInterface() {propertyValues {Curve {keys [<at 0 0>,<at 1 1>]}}})))
  ]
  rendererClasses [&NiagaraMeshRendererProperties]
  renderers [{rendererClass=&NiagaraMeshRendererProperties, rendererIndex=0}]
 }
 Horn_Air_Burst {
  emitterUpdateScript [
   &EmitterState()
   &SpawnRate(
    SpawnRate=&Lerp_Float(
     A=3
     B=4.5
     Alpha=&FloatFromCurve(
      FloatCurve=DataInterface() {propertyValues {Curve {keys [
          <at 0 0 RCIM_Cubic>
          {arriveTangent=0.2115, interpMode=RCIM_Cubic, leaveTangent=0.2115, time=3.0268, value=0.2044}
          {interpMode=RCIM_Cubic, tangentMode=RCTM_User, time=4.7274, value=1}
          {interpMode=RCIM_Cubic, tangentMode=RCTM_User, time=7.7507, value=1}
          {arriveTangent=-0.4452, interpMode=RCIM_Cubic, leaveTangent=-0.4452, time=8.8844, value=0.3676}
          <at 9.9967 0>
         ]}}}
      CurveIndex=$System.LoopedAge
     )
    )
    Use Spawn Probability=true
   )
  ]
  particleSpawnScript [
   &InitializeParticle(
    Color=<rgba 1 0.8221 0.5 0.4>
    Lifetime=0.125
    Mesh Scale=&Multiply_Vector(
     A=&Multiply_Vector(
      A=&MakeVector(X=1.5, Y=1.25, Z=&RandomRangeFloat(Minimum=1.1, Maximum=1.25))
      B=&VectorFromFloat(Value=&RandomRangeFloat(Minimum=0.5, Maximum=0.75))
     )
     B=&VectorFromFloat(Value=1)
    )
    Module.Write Color=true
    Module.Write Lifetime=true
    Module.Write Mass=true
    Module.Write Material Random=true
    Module.Write Position=false
    Module.Write Scale=true
    Module.Write SpriteRotation=false
    Module.Write SpriteSize=false
    Position=<0 0 10>
   )
   /SetVariables(
    Particles.MeshScale_02:Vector3f=<1 1 1>
    Particles.XOffset:NiagaraFloat=&RandomRangeFloat(Minimum=0, Maximum=1)
    Particles.XOffset_02:NiagaraFloat=&RandomRangeFloat(Minimum=0, Maximum=1)
   ) {isSetParameters=true, module=SetVariables_#3}
   &InitialMeshOrientation(
    Orientation Coordinate Space=::Local
    Orientation Vector=<0 0 1>
    Rotation=&RandomRangeVector(Minimum=<0 0 0>, Maximum=<0 0 1>)
    Use Orientation Vector=true
    Use Rotation Vector=true
   )
  ]
  particleUpdateScript [
   &ParticleState()
   &DynamicMaterialParameters(
    Index 0 Param 1=1
    Index 0 Param 2=0.04
    Index 0 Param 3=20
    Index 0 Param 4=1
    Index 1 Param 1=&Add_Float(
     A=$Particles.Initial.XOffset
     B=&FloatFromCurve(FloatCurve=DataInterface() {propertyValues {Curve {keys [<at 0 0>,<at 1 1>]}}}, Scale Curve=2.5)
    )
    Index 1 Param 2=&RandomRangeFloat(Minimum=0, Maximum=1)
    Index 2 Param 1=&Add_Float(
     A=$Particles.Initial.XOffset_02
     B=&FloatFromCurve(FloatCurve=DataInterface() {propertyValues {Curve {keys [<at 0 0>,<at 1 1>]}}}, Scale Curve=2.5)
    )
    Index 2 Param 2=&RandomRangeFloat(Minimum=0, Maximum=1)
    Module.Param0WriteEnabled=true
    Module.Param0WriteEnabled1=true
    Module.Param0WriteEnabled2=true
    Module.Param1WriteEnabled=false
    Module.Param1WriteEnabled1=true
    Module.Param1WriteEnabled2=true
    Module.Param2WriteEnabled=false
    Module.Param2WriteEnabled1=false
    Module.Param2WriteEnabled2=true
    Module.Param3WriteEnabled=true
    Module.Param3WriteEnabled1=false
    Module.Param3WriteEnabled2=true
    Write Parameter Index 1=true
    Write Parameter Index 2=true
   ) {module=DynamicMaterialParameters001}
   &ScaleColor(
    Scale Alpha=&FloatFromCurve(
     FloatCurve=DataInterface() {propertyValues {Curve {keys [
         {arriveTangent=3.2844, interpMode=RCIM_Cubic, leaveTangent=3.2844, tangentMode=RCTM_User, time=0, value=0}
         <at 0.2049 0.8535 RCIM_Cubic>
         {arriveTangent=-3.5444, interpMode=RCIM_Cubic, leaveTangent=-3.5444, tangentMode=RCTM_User, time=1, value=0}
        ]}}}
     Scale Curve=&MakeCustomFloatFromBool(
      Bool=&IsPlatformSetActive(New Platform Set=DataInterface() {propertyValues {Platforms {qualityLevelMask=-32, deviceProfileStates [{profileName=Switch, qualityLevelMask=4294967295, setQualityLevelMask=4294967295}]}}})
      True Float=8
      False Float=1
     )
    )
    ScaleA=true
    ScaleRGB=false
   )
   &UpdateMeshOrientation(
    Rotation Coordinate Space=::NewEnumerator3
    Rotation Rate=&RandomRangeFloat(Minimum=-1, Maximum=1)
    Rotation Vector=<0 0 1>
   )
   &ScaleMeshSize(Scale Factor=&VectorFromFloat(Value=&FloatFromCurve(FloatCurve=DataInterface() {propertyValues {Curve {keys [
          {arriveTangent=5.5595, interpMode=RCIM_Cubic, leaveTangent=5.5595, tangentMode=RCTM_User, time=0, value=0}
          <at 0.5 1 RCIM_Cubic>
          {arriveTangent=-4.7137, interpMode=RCIM_Cubic, leaveTangent=-4.7137, tangentMode=RCTM_User, time=1, value=0}
         ]}}})))
  ]
  rendererClasses [&NiagaraMeshRendererProperties]
  renderers [{rendererClass=&NiagaraMeshRendererProperties, rendererIndex=0}]
 }
 Lines {
  emitterUpdateScript [
   &EmitterState()
   &SpawnRate(
    SpawnRate=&Multiply_Float(
     A=50
     B=&FloatFromCurve(
      FloatCurve=DataInterface() {propertyValues {Curve {keys [
          {arriveTangent=0.0741, interpMode=RCIM_Cubic, leaveTangent=0.0741, tangentMode=RCTM_User, time=0, value=0.1448}
          {arriveTangent=0.3055, interpMode=RCIM_Cubic, leaveTangent=0.3055, tangentMode=RCTM_User, time=3.3246, value=0.5479}
          {arriveTangent=-0.0538, interpMode=RCIM_Cubic, leaveTangent=-0.0538, time=6.9599, value=0.9432}
          {arriveTangent=-0.3869, interpMode=RCIM_Cubic, leaveTangent=-0.3869, tangentMode=RCTM_User, time=8.909, value=0.2473}
          {arriveTangent=-0.0357, interpMode=RCIM_Cubic, leaveTangent=-0.0357, tangentMode=RCTM_User, time=10, value=0.1281}
         ]}}}
      CurveIndex=$System.LoopedAge
     )
    )
    Use Spawn Probability=true
   )
  ]
  particleSpawnScript [
   &DynamicMaterialParameters(
    Index 0 Param 1=&RandomRangeFloat(Minimum=0, Maximum=0.5, Evaluation Type=::NewEnumerator1)
    Index 0 Param 2=&RandomRangeFloat(Minimum=1, Maximum=8, Evaluation Type=::NewEnumerator1)
    Index 0 Param 3=1
    Index 0 Param 4=1
    Module.Param0WriteEnabled=true
    Module.Param1WriteEnabled=true
    Module.Param2WriteEnabled=false
    Module.Param3WriteEnabled=false
   )
   &InitializeParticle(
    Color=<rgba 1 0.8081 0.326 0.05>
    Lifetime=&RandomRangeFloat(Minimum=0.5, Maximum=0.25)
    Module.Write Color=true
    Module.Write Lifetime=true
    Module.Write Mass=true
    Module.Write Material Random=true
    Module.Write Position=false
    Module.Write Scale=false
    Module.Write SpriteRotation=true
    Module.Write SpriteSize=true
    Sprite Rotation=&NormalizedAngleToDegrees(NormalizedAngle=&RandomRangeFloat(Minimum=-0.1, Maximum=0.1, Evaluation Type=::NewEnumerator1))
    Sprite Size=<3 20>
   )
   &AddVelocityInCone(Cone Angle=75, Module.Use Velocity Falloff On Cone Axis=false, Velocity Distribution Along Cone Axis=0, Velocity Strength=200)
  ]
  particleUpdateScript [
   &ParticleState()
   &ScaleColor(
    Scale Alpha=&FloatFromCurve(
     FloatCurve=DataInterface() {propertyValues {Curve {keys [<at 0.2839 0>,<at 0.3996 1>,<at 1 0>]}}}
     Scale Curve=&MakeCustomFloatFromBool(
      Bool=&IsPlatformSetActive(New Platform Set=DataInterface() {propertyValues {Platforms {qualityLevelMask=-32, deviceProfileStates [{profileName=Switch, qualityLevelMask=4294967295, setQualityLevelMask=4294967295}]}}})
      True Float=20
      False Float=1
     )
    )
    ScaleA=true
    ScaleRGB=true
   )
   &SolveForcesAndVelocity(Clamp Velocity=false, Limit Acceleration=false)
   &ScaleSpriteSize(Scale Factor=&Vector2DFromFloat(Value=&FloatFromCurve(FloatCurve=DataInterface() {propertyValues {Curve {keys [
          <at 0 0 RCIM_Cubic>
          {interpMode=RCIM_Cubic, tangentMode=RCTM_User, time=0.0837, value=1.0126}
          {arriveTangent=-0.8322, interpMode=RCIM_Cubic, leaveTangent=-0.8322, time=0.25, value=1}
          <at 1 0.25 RCIM_Cubic>
         ]}}})))
  ]
  rendererClasses [&NiagaraSpriteRendererProperties]
  renderers [{rendererClass=&NiagaraSpriteRendererProperties}]
 }
 MusicSheet_Trail_Velocity {
  emitterUpdateScript [
   &EmitterState()
   &SpawnRate(SpawnRate=45, Use Spawn Probability=false)
  ]
  particleSpawnScript [
   &InitializeRibbon(
    Color=<rgba 0.325 0.2472 0.0517 0.333>
    Lifetime=0.5
    Module.Write Color=true
    Module.Write Lifetime=true
    Module.Write Mass=true
    Module.Write Material Random=true
    Module.Write Position=true
    Module.Write RibbonTwist=true
    Module.Write RibbonWidth=true
    Position:NiagaraPosition=$Engine.Owner.Position
    Ribbon Twist:NiagaraFloat=&Sine(Normalized Angle=$Emitter.Age, Period=1, Scale=0.25, Bias=0)
    Ribbon Width=40
   )
   &DynamicMaterialParameters(Module.Param0WriteEnabled=true, Module.Param1WriteEnabled=true, Module.Param2WriteEnabled=true, Module.Param3WriteEnabled=true)
   &InheritVelocity(Module.Limit Speed=true, Velocity Limit=10000, Velocity Scale=<1.15 1.15 1.15>)
   &AddVelocity(Velocity=&Multiply_Vector(A=<500 0 0>, B=&VectorFromFloat(Value=1.5)))
   /SetVariables() {isSetParameters=true}
  ]
  particleUpdateScript [
   &ParticleState()
   &Drag(Drag=10, Use Linear Drag=true, Use Rotational Drag=false)
   &SolveForcesAndVelocity(Clamp Velocity=false, Limit Acceleration=false)
   &ScaleColor(
    Scale Alpha=&FloatFromCurve(
     FloatCurve=DataInterface() {propertyValues {Curve {keys [
         {arriveTangent=-0.0391, interpMode=RCIM_Cubic, time=0.0866, value=0}
         {arriveTangent=-0.8017, interpMode=RCIM_Cubic, leaveTangent=-0.8017, tangentMode=RCTM_User, time=0.1316, value=1}
         {arriveTangent=-1.2396, interpMode=RCIM_Cubic, leaveTangent=-1.2396, tangentMode=RCTM_User, time=0.7635, value=0.4539}
         {arriveTangent=-0.0191, interpMode=RCIM_Cubic, leaveTangent=-0.0191, tangentMode=RCTM_User, time=1, value=0}
        ]}}}
     Scale Curve=&MakeCustomFloatFromBool(
      Bool=&IsPlatformSetActive(New Platform Set=DataInterface() {propertyValues {Platforms {qualityLevelMask=-32, deviceProfileStates [{profileName=Switch, qualityLevelMask=4294967295, setQualityLevelMask=4294967295}]}}})
      True Float=7.5
      False Float=1
     )
    )
    ScaleA=true
    ScaleRGB=false
   )
   &ScaleRibbonWidth(Ribbon Width Scale=&FloatFromCurve(FloatCurve=DataInterface() {propertyValues {Curve {keys [
         <at 0.0165 0.5084 RCIM_Cubic>
         {arriveTangent=0.7181, interpMode=RCIM_Cubic, leaveTangent=0.7181, tangentMode=RCTM_User, time=0.4434, value=0.8163}
         {interpMode=RCIM_Cubic, tangentMode=RCTM_User, time=1, value=1}
        ]}}})) {module=ScaleRibbonWidth001}
   /SetVariables(Particles.Wave Position:Vector3f=&Add_Vector(
     A=Linked() {linkedVariable {name=Particles.Position, type=NiagaraPosition}}
     B=&MakeVector(
      X=0
      Y=0
      Z=&Sine(
       Normalized Angle=&Add_Float(A=$Emitter.Age, B=&Multiply_Float(A=$Particles.NormalizedAge, B=-0.5))
       Period=0.35
       Scale=&FloatFromCurve(
        FloatCurve=DataInterface() {propertyValues {Curve {keys [<at 0 0>, {arriveTangent=1, interpMode=RCIM_Cubic, leaveTangent=1, time=0.2304, value=0.7182}, <at 1 1>]}}}
        Scale Curve=10
       )
       Bias=0
      )
     )
    )) {isSetParameters=true}
  ]
  rendererClasses [&NiagaraRibbonRendererProperties]
  renderers [{rendererClass=&NiagaraRibbonRendererProperties}]
 }
 Music_Notes {
  emitterUpdateScript [
   &EmitterState()
   &SpawnRate(
    SpawnRate=&Lerp_Float(
     A=3
     B=4.5
     Alpha=&FloatFromCurve(
      FloatCurve=DataInterface() {propertyValues {Curve {keys [
          <at 0 0 RCIM_Cubic>
          {arriveTangent=0.2115, interpMode=RCIM_Cubic, leaveTangent=0.2115, time=3.0268, value=0.2044}
          {interpMode=RCIM_Cubic, tangentMode=RCTM_User, time=4.7274, value=1}
          {interpMode=RCIM_Cubic, tangentMode=RCTM_User, time=7.7507, value=1}
          {arriveTangent=-0.5381, interpMode=RCIM_Cubic, leaveTangent=-0.5381, tangentMode=RCTM_User, time=8.8844, value=0.3676}
          <at 9.9967 0>
         ]}}}
      CurveIndex=$System.LoopedAge
     )
    )
    Use Spawn Probability=true
   )
  ]
  particleSpawnScript [
   /Niagara/Modules/Update/SubUV/SubUVAnimation(SubUV Animation Mode=::NewEnumerator1)
   &DynamicMaterialParameters(
    Index 0 Param 1=&Multiply_Float(A=$System.Age, B=2)
    Index 0 Param 2=&RandomRangeFloat(Minimum=5, Maximum=8, Evaluation Type=::NewEnumerator1)
    Index 0 Param 3=1
    Index 0 Param 4=1
    Module.Param0WriteEnabled=true
    Module.Param1WriteEnabled=true
    Module.Param2WriteEnabled=false
    Module.Param3WriteEnabled=false
   )
   &InitializeParticle(
    Color=<rgba 1 0.7605 0.159 1>
    Lifetime=&RandomRangeFloat(Minimum=0.6, Maximum=0.5)
    Module.Write Color=true
    Module.Write Lifetime=true
    Module.Write Mass=true
    Module.Write Material Random=true
    Module.Write Position=false
    Module.Write Scale=false
    Module.Write SpriteRotation=true
    Module.Write SpriteSize=true
    Sprite Rotation=&NormalizedAngleToDegrees(NormalizedAngle=&RandomRangeFloat(Minimum=-0.1, Maximum=0.1, Evaluation Type=::NewEnumerator1))
    Sprite Size=&Vector2DFromFloat(Value=25)
   ) {enabled=false}
   /Niagara/Modules/Spawn/Initialization/V2/InitializeParticle(
    Color=<rgba 1 0.7605 0.159 1>
    Lifetime=&RandomRangeFloat(Minimum=0.6, Maximum=0.5)
    Mass Mode=::NewEnumerator1
    Sprite Rotation Angle=&RandomRangeFloat(Minimum=-0.1, Maximum=0.1)
    Sprite Rotation Mode=::NewEnumerator1
    Sprite Size=&Vector2DFromFloat(Value=25)
    Sprite Size Mode=::NewEnumerator4
    UsePositionOffset=false
    Write Material Random=true
   ) {module=InitializeParticle001}
   &AddVelocity(
    Scale Added Velocity=&VectorFromFloat(Value=1.5)
    Velocity=&RandomRangeVector(Minimum=<500 0 100>, Maximum=<500 1 -100>)
   )
   &InheritVelocity(Module.Limit Speed=true, Velocity Limit=10000, Velocity Scale=<1.15 1.15 1.15>)
   /SetVariables(Particles.CameraOffset=-1, Particles.Light Color:LinearColor=<rgba 1 0.9431 0.8 220>) {isSetParameters=true}
  ]
  particleUpdateScript [
   &ParticleState()
   &ScaleColor(
    Scale Alpha=&FloatFromCurve(FloatCurve=DataInterface() {propertyValues {Curve {keys [<at 0 1>,<at 1 0>]}}})
    Scale RGB=&VectorFromFloat(Value=&FloatFromCurve(FloatCurve=DataInterface() {propertyValues {Curve {keys [
          <at 0 0 RCIM_Cubic>
          {interpMode=RCIM_Cubic, tangentMode=RCTM_User, time=0.1062, value=6}
          {arriveTangent=-7.4341, interpMode=RCIM_Cubic, tangentMode=RCTM_Break, time=0.2671, value=0.9767}
          <at 1 1 RCIM_Cubic>
         ]}}}))
    ScaleA=true
    ScaleRGB=true
   )
   &Drag(Drag=10, Use Linear Drag=true, Use Rotational Drag=false)
   &SolveForcesAndVelocity(Clamp Velocity=false, Limit Acceleration=false)
   &ScaleSpriteSize(Scale Factor=&Vector2DFromFloat(Value=&FloatFromCurve(FloatCurve=DataInterface() {propertyValues {Curve {keys [
          <at 0 0 RCIM_Cubic>
          {interpMode=RCIM_Cubic, tangentMode=RCTM_User, time=0.1605, value=1.2189}
          {interpMode=RCIM_Cubic, tangentMode=RCTM_User, time=0.2616, value=0.4464}
          <at 0.3931 0.8767>
          {arriveTangent=-1.9685, interpMode=RCIM_Cubic, leaveTangent=-1.9685, tangentMode=RCTM_User, time=0.7815, value=0.5908}
          <at 1 0 RCIM_Cubic>
         ]}}})))
   /SetVariables(
    Particles.LightEnabled:NiagaraBool=&SetBoolByFloatComparison(A=$Particles.NormalizedAge, B=0.08, Comparison Type=::NewEnumerator7)
    Particles.SpriteRotation:NiagaraFloat=&Sine(
     Normalized Angle=&Add_Float(A=$Emitter.Age, B=&RandomRangeFloat(Minimum=0, Maximum=1))
     Period=1
     Scale=30
     Bias=0
    )
   ) {isSetParameters=true}
  ]
  rendererClasses [&NiagaraSpriteRendererProperties, &NiagaraLightRendererProperties]
  renderers [
   {rendererClass=&NiagaraSpriteRendererProperties}
   {rendererClass=&NiagaraLightRendererProperties, rendererIndex=1}
  ]
 }
}
summary {
 systemName=NS_Phonograph_Notes
 emitters [
  {emitterName=Music_Notes, rendererClasses [&NiagaraSpriteRendererProperties, &NiagaraLightRendererProperties]}
  {emitterName=Lines, rendererClasses [&NiagaraSpriteRendererProperties]}
  {emitterName=MusicSheet_Trail_Velocity, rendererClasses [&NiagaraRibbonRendererProperties]}
  {emitterName=Horn_Air, rendererClasses [&NiagaraMeshRendererProperties]}
  {emitterName=Horn_Air_Burst, rendererClasses [&NiagaraMeshRendererProperties]}
 ]
 userVariables []
}
